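# Physics asset for the F_MED_ShiverFlame_Ace female character body, covering a dynamic pouch bone chain and the right leg.
# Capsules on the right thigh and calf sit alongside capsules on the pouch master and four pouch bones.
# Despite its "left" naming, the pouch master is tightly constrained to the right thigh, with generally looser swing limits down the chain.
# The calf-to-thigh constraint leaves swing and twist free.
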PhysicsAsset /BRCosmetics/Characters/Player/Female/Medium/Bodies/F_MED_ShiverFlame_Ace/Meshes/F_MED_ShiverFlame_Ace_Physics
bodies {
 calf_r [{
   shapeType=Capsule
   length=30.5721
   radius=6.7289
   shapeName=calf_r_capsule
   center=<17.9226 4.0914 0.8715>
   rotation=<rot 82.3505 177.5476 -177.9687>
  }]
 dyn_left_pouch_01 [{shapeType=Capsule, length=6.3015, radius=0.801, shapeName=dyn_left_pouch_01_capsule, center=<3.1433 -1 0>, rotation=<rot 90 0 0>}]
 dyn_left_pouch_02 [{shapeType=Capsule, length=6.2821, radius=2.8968, shapeName=dyn_left_pouch_02_capsule, center=<3.8187 -0.1287 -1.8996>, rotation=<rot 90 0 0>}]
 dyn_left_pouch_03 [{shapeType=Capsule, length=7.8394, radius=1.664, shapeName=dyn_left_pouch_03_capsule, center=<4.3016 0 0>, rotation=<rot 90 0 0>}]
 dyn_left_pouch_04 [{
   shapeType=Capsule
   length=5.0063
   radius=2.221
   shapeName=dyn_left_pouch_04_capsule
   center=<6.0979 0 -0.8365>
   rotation=<rot 90 170.5377 80.5376>
  }]
 dyn_left_pouch_MSTR [{shapeType=Capsule, length=1, radius=1, shapeName=dyn_left_pouch_MSTR_capsule, rotation=<rot 0 90 0>}]
 thigh_r [{
   shapeType=Capsule
   length=27.8069
   radius=8.9308
   shapeName=thigh_r_capsule
   center=<22.0771 -0.8619 -0.7657>
   rotation=<rot 0 -78.5017 -84.0416>
  }]
}
constraints [
 {bone1Name=dyn_left_pouch_MSTR, bone2Name=thigh_r, swing1LimitDegrees=10, swing2LimitDegrees=10, twistLimitDegrees=1}
 {bone1Name=dyn_left_pouch_02, bone2Name=dyn_left_pouch_01, swing1LimitDegrees=60, swing2LimitDegrees=30, twistLimitDegrees=1}
 {bone1Name=dyn_left_pouch_01, bone2Name=dyn_left_pouch_MSTR, swing1LimitDegrees=45, swing2LimitDegrees=30, twistLimitDegrees=1}
 {
  bone1Name=calf_r
  bone2Name=thigh_r
  swing1LimitDegrees=45
  swing1Motion=Free
  swing2LimitDegrees=45
  swing2Motion=Free
  twistLimitDegrees=45
  twistMotion=Free
 }
 {bone1Name=dyn_left_pouch_03, bone2Name=dyn_left_pouch_02, swing1LimitDegrees=60, swing2LimitDegrees=45, twistLimitDegrees=10}
 {bone1Name=dyn_left_pouch_04, bone2Name=dyn_left_pouch_03, swing1LimitDegrees=60, swing2LimitDegrees=45, twistLimitDegrees=10}
]
