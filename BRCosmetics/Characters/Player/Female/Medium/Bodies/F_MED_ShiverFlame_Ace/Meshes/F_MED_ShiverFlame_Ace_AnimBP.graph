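# Post-copy AnimBlueprint for the F_MED_ShiverFlame_Ace female body cosmetic, layering secondary motion and fixups onto the parent mesh pose.
# The copied pose runs through RigidBody with overridden gravity, a parachute- and emote-aware Control Rig, attachment and discman offsets, and a left-pauldron constraint.
# Chin and jaw bones replicate transforms read each update from the Head part mesh.
# Events also derive gravity, movement and helmet state.
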AnimBlueprint /BRCosmetics/Characters/Player/Female/Medium/Bodies/F_MED_ShiverFlame_Ace/Meshes/F_MED_ShiverFlame_Ace_AnimBP
target_skeleton=&/F_MED_ShiverFlame_Ace_Skeleton
graphs [
 AnimGraph {nodes [
   n1 = RigidBody!(
    ComponentPose=Local To Component(LocalPose=Copy Pose From Mesh() {bUseAttachedParent=true})
    Alpha
    OverrideWorldGravity:Vector=Get Override World Gravity() -> (Override World Gravity:Vector)
    ExternalForce
   ) -> (Pose) {
    bClampLinearTranslationLimitToRefPose=true
    bOverrideWorldGravity=true
    overridePhysicsAsset=&/F_MED_ShiverFlame_Ace_Physics
    simulationSpace=BaseBoneSpace
    alphaBoolBlend {blendOption=Linear}
    baseBoneRef {boneName=pelvis}
    componentAppliedLinearAccClamp=<2000 2000 2000>
    componentLinearAccScale=<0.3 0.3 0.3>
    componentLinearVelScale=<0.3 0.3 0.3>
    overrideWorldGravity {z=-980}
   }
   n2 = Control Rig!(Source=Component To Local(ComponentPose=@n1), Alpha, isParachute:bool=Get bIsParachuteOpened(), isEmote:bool=Get bIsPlayingEmote()) -> (Pose) {
    alphaBoolBlend {blendOption=Linear}
    controlRigAssetReference {blueprintRigClass=&/F_MED_ShiverFlame_Ace_CtrlRig.F_MED_ShiverFlame_Ace_CtrlRig_C}
   }
   n3 = "Transform (Modify) Bone - Bone: attach_backpack"!(ComponentPose=Local To Component(LocalPose=@n2), bAlphaBoolEnabled:bool=True, Translation="0,-1.3,0", Rotation="0,0,-13", Scale="1,1,1") -> (Pose) {
    alphaInputType=Bool
    rotationMode=BMM_Additive
    rotationSpace=BCS_BoneSpace
    translationMode=BMM_Additive
    translationSpace=BCS_BoneSpace
    alphaBoolBlend {blendOption=Linear}
    boneToModify {boneName=attach_backpack}
    rotation {roll=-13}
    translation {y=-1.3}
   }
   n4 = "Transform (Modify) Bone - Bone: attach_cape"!(ComponentPose=@n3, bAlphaBoolEnabled:bool=True, Translation="0,-3,0", Rotation="0,0,0", Scale="1,1,1") -> (Pose) {
    alphaInputType=Bool
    rotationMode=BMM_Additive
    translationMode=BMM_Additive
    translationSpace=BCS_BoneSpace
    alphaBoolBlend {blendOption=Linear}
    boneToModify {boneName=attach_cape}
    translation {y=-3}
   }
   n5 = "Transform (Modify) Bone - Bone: discman"(ComponentPose=@n4, Alpha=1, Translation="-1,-2,0.2", Rotation="20,-20,0", Scale="1,1,1") {
    rotationMode=BMM_Additive
    translationMode=BMM_Additive
    translationSpace=BCS_BoneSpace
    alphaBoolBlend {blendInTime=0.3, blendOption=CubicInOut, blendOutTime=0.3}
    alphaScaleBiasClamp {bInterpResult=true, interpSpeedDecreasing=5, interpSpeedIncreasing=5}
    boneToModify {boneName=discman}
    rotation {pitch=20, yaw=-20}
    translation=<-1 -2 0.2>
   }
   n6 = "Transform (Modify) Bone - Bone: discman"!(ComponentPose=@n5, bAlphaBoolEnabled:bool=True, Translation="-3,0,1", Rotation="30,0,0", Scale="1,1,1") -> (Pose) {
    alphaInputType=Bool
    rotationMode=BMM_Additive
    rotationSpace=BCS_BoneSpace
    translationMode=BMM_Additive
    translationSpace=BCS_BoneSpace
    alphaBoolBlend {blendInTime=0.5, blendOption=Linear, blendOutTime=0.5}
    boneToModify {boneName=discman}
    rotation {pitch=30}
    translation {x=-3, z=1}
   }
   n7 = "Transform (Modify) Bone - Bone: discman"(ComponentPose=@n6, bAlphaBoolEnabled=True, Translation="-1.563,-8.454,-4.885", Rotation="3.259,-36.633,-21.777", Scale="1,1,1") {
    alphaInputType=Bool
    rotationMode=BMM_Additive
    translationMode=BMM_Additive
    translationSpace=BCS_BoneSpace
    alphaBoolBlend {blendInTime=0.3, blendOption=CubicInOut, blendOutTime=0.3}
    boneToModify {boneName=discman}
    rotation=<rot 3.259 -36.633 -21.777>
    translation=<-1.563 -8.454 -4.885>
   }
   n8 = "Constraint - pauldron_l (PP)"!(ComponentPose=@n7, bAlphaBoolEnabled:bool=True, ConstraintWeights_0=0.8, ConstraintWeights_1:float=0.2) -> (Pose) {
    alphaInputType=Bool
    alphaBoolBlend {blendInTime=0.3, blendOption=Linear, blendOutTime=0.3}
    boneToModify {boneName=pauldron_l}
    constraintSetup [{targetBone {boneName=clavicle_l}}, {targetBone {boneName=upperarm_l}}]
    constraintWeights [1,1]
   }
   n9 = "Transform (Modify) Bone - Bone: pauldron_l"!(ComponentPose=@n8, bAlphaBoolEnabled:bool=True, Translation="0.007,0.012,1.8", Rotation="0,13,0", Scale="1,1,1") -> (Pose) {
    alphaInputType=Bool
    rotationMode=BMM_Additive
    translationMode=BMM_Additive
    translationSpace=BCS_BoneSpace
    alphaBoolBlend {blendInTime=0.3, blendOption=CubicInOut, blendOutTime=0.3}
    boneToModify {boneName=pauldron_l}
    rotation {yaw=13}
    translation=<0.007 0.012 1.8>
   }
   n10 = "Transform (Modify) Bone - Bone: FACIAL_C_UnderChin"!(
    ComponentPose=@n9
    bAlphaBoolEnabled:bool=True
    Translation=Get FACIAL_C_UnderChin_Translation() -> (FACIAL_C_UnderChin_Translation:Vector)
    Rotation=Get FACIAL_C_UnderChin_Rotation() -> (FACIAL_C_UnderChin_Rotation:Rotator)
    Scale="1,1,1"
   ) -> (Pose) {
    alphaInputType=Bool
    rotationMode=BMM_Replace
    rotationSpace=BCS_ParentBoneSpace
    translationMode=BMM_Replace
    translationSpace=BCS_ParentBoneSpace
    alphaBoolBlend {blendOption=Linear}
    boneToModify {boneName=FACIAL_C_UnderChin}
   }
   n11 = "Transform (Modify) Bone - Bone: FACIAL_L_UnderChin"!(
    ComponentPose=@n10
    bAlphaBoolEnabled:bool=True
    Translation=Get FACIAL_L_UnderChin_Translation() -> (FACIAL_L_UnderChin_Translation:Vector)
    Rotation=Get FACIAL_L_UnderChin_Rotation() -> (FACIAL_L_UnderChin_Rotation:Rotator)
    Scale="1,1,1"
   ) -> (Pose) {
    alphaInputType=Bool
    rotationMode=BMM_Replace
    rotationSpace=BCS_ParentBoneSpace
    translationMode=BMM_Replace
    translationSpace=BCS_ParentBoneSpace
    alphaBoolBlend {blendOption=Linear}
    boneToModify {boneName=FACIAL_L_UnderChin}
   }
   n12 = "Transform (Modify) Bone - Bone: FACIAL_R_UnderChin"!(
    ComponentPose=@n11
    bAlphaBoolEnabled:bool=True
    Translation=Get FACIAL_R_UnderChin_Translation() -> (FACIAL_R_UnderChin_Translation:Vector)
    Rotation=Get FACIAL_R_UnderChin_Rotation() -> (FACIAL_R_UnderChin_Rotation:Rotator)
    Scale="1,1,1"
   ) -> (Pose) {
    alphaInputType=Bool
    rotationMode=BMM_Replace
    rotationSpace=BCS_ParentBoneSpace
    translationMode=BMM_Replace
    translationSpace=BCS_ParentBoneSpace
    alphaBoolBlend {blendOption=Linear}
    boneToModify {boneName=FACIAL_R_UnderChin}
   }
   n13 = "Transform (Modify) Bone - Bone: FACIAL_L_Jawline"!(
    ComponentPose=@n12
    bAlphaBoolEnabled:bool=True
    Translation=Get FACIAL_L_Jawline_Translation() -> (FACIAL_L_Jawline_Translation:Vector)
    Rotation=Get FACIAL_L_Jawline_Rotation() -> (FACIAL_L_Jawline_Rotation:Rotator)
    Scale="1,1,1"
   ) -> (Pose) {
    alphaInputType=Bool
    rotationMode=BMM_Replace
    rotationSpace=BCS_ParentBoneSpace
    translationMode=BMM_Replace
    translationSpace=BCS_ParentBoneSpace
    alphaBoolBlend {blendOption=Linear}
    boneToModify {boneName=FACIAL_L_Jawline}
   }
   n14 = "Transform (Modify) Bone - Bone: FACIAL_R_Jawline"!(
    ComponentPose=@n13
    bAlphaBoolEnabled:bool=True
    Translation=Get FACIAL_R_Jawline_Translation() -> (FACIAL_R_Jawline_Translation:Vector)
    Rotation=Get FACIAL_R_Jawline_Rotation() -> (FACIAL_R_Jawline_Rotation:Rotator)
    Scale="1,1,1"
   ) -> (Pose) {
    alphaInputType=Bool
    rotationMode=BMM_Replace
    rotationSpace=BCS_ParentBoneSpace
    translationMode=BMM_Replace
    translationSpace=BCS_ParentBoneSpace
    alphaBoolBlend {blendOption=Linear}
    boneToModify {boneName=FACIAL_R_Jawline}
   }
   n15 = "Transform (Modify) Bone - Bone: FACIAL_C_Jaw"!(
    ComponentPose=@n14
    bAlphaBoolEnabled:bool=True
    Translation=Get FACIAL_C_Jaw_Tranform() -> (FACIAL_C_Jaw_Tranform:Vector)
    Rotation=Get FACIAL_C_Jaw_Rotation() -> (FACIAL_C_Jaw_Rotation:Rotator)
    Scale="1,1,1"
   ) -> (Pose) {
    alphaInputType=Bool
    rotationMode=BMM_Replace
    rotationSpace=BCS_ParentBoneSpace
    translationMode=BMM_Replace
    translationSpace=BCS_ParentBoneSpace
    alphaBoolBlend {blendOption=Linear}
    boneToModify {boneName=FACIAL_C_Jaw}
   }
   n16 = "Transform (Modify) Bone - Bone: FACIAL_C_Jawline"!(
    ComponentPose=@n15
    bAlphaBoolEnabled:bool=True
    Translation=Get FACIAL_C_Jawline_Translation() -> (FACIAL_C_Jawline_Translation:Vector)
    Rotation=Get FACIAL_C_Jawline_Rotation() -> (FACIAL_C_Jawline_Rotation:Rotator)
    Scale="1,1,1"
   ) -> (Pose) {
    alphaInputType=Bool
    rotationMode=BMM_Replace
    rotationSpace=BCS_ParentBoneSpace
    translationMode=BMM_Replace
    translationSpace=BCS_ParentBoneSpace
    alphaBoolBlend {blendOption=Linear}
    boneToModify {boneName=FACIAL_C_Jawline}
   }
   Output Pose(Result=Component To Local(ComponentPose=@n16)) {layerGroup=None}
  ]}
 EventGraph {nodes [
   Event BlueprintUpdateAnimation()
   n1 = Get BodyMesh() -> (BodyMesh:Skeletal Mesh Component Budgeted Object Reference)
   > Is Valid(InputObject=@n1)
   n2 = GetSocketBoneName!(self:Skinned Mesh Component Object Reference=@n1, InSocketName:Name=FACIAL_C_UnderChin) -> (ReturnValue:Name=None)
   > n3 = Set isFaceBones(isFaceBones:bool="Not Equal Exactly (String)"(A="To String (Name)"(InName=@n2), B=None)) -> (Output_Get:bool)
   > Branch(Condition=@n3)
   n4 = Get BodyMesh()
   n5 = BreakTransform(InTransform=GetSocketTransform(self=@n4, InSocketName=FACIAL_C_UnderChin, TransformSpace=RTS_ParentBoneSpace))
   > Set FACIAL_C_UnderChin_Translation(FACIAL_C_UnderChin_Translation:Vector=@n5) -> (Output_Get:Vector)
   > Set FACIAL_C_UnderChin_Rotation(FACIAL_C_UnderChin_Rotation:Rotator=@n5.Rotation) -> (Output_Get:Rotator)
   n6 = BreakTransform(InTransform=GetSocketTransform(self=@n4, InSocketName=FACIAL_L_UnderChin, TransformSpace=RTS_ParentBoneSpace))
   > Set FACIAL_L_UnderChin_Translation(FACIAL_L_UnderChin_Translation:Vector=@n6) -> (Output_Get:Vector)
   > Set FACIAL_L_UnderChin_Rotation(FACIAL_L_UnderChin_Rotation:Rotator=@n6.Rotation) -> (Output_Get:Rotator)
   n7 = BreakTransform(InTransform=GetSocketTransform(self=@n4, InSocketName=FACIAL_R_UnderChin, TransformSpace=RTS_ParentBoneSpace))
   > Set FACIAL_R_UnderChin_Translation(FACIAL_R_UnderChin_Translation:Vector=@n7) -> (Output_Get:Vector)
   > Set FACIAL_R_UnderChin_Rotation(FACIAL_R_UnderChin_Rotation:Rotator=@n7.Rotation) -> (Output_Get:Rotator)
   n8 = BreakTransform(InTransform=GetSocketTransform(self=@n4, InSocketName=FACIAL_C_Jaw, TransformSpace=RTS_ParentBoneSpace))
   > Set FACIAL_C_Jaw_Tranform(FACIAL_C_Jaw_Tranform:Vector=@n8) -> (Output_Get:Vector)
   > Set FACIAL_C_Jaw_Rotation(FACIAL_C_Jaw_Rotation:Rotator=@n8.Rotation) -> (Output_Get:Rotator)
   n9 = BreakTransform(InTransform=GetSocketTransform(self=@n4, InSocketName=FACIAL_C_Jawline, TransformSpace=RTS_ParentBoneSpace))
   > Set FACIAL_C_Jawline_Translation(FACIAL_C_Jawline_Translation:Vector=@n9) -> (Output_Get:Vector)
   > Set FACIAL_C_Jawline_Rotation(FACIAL_C_Jawline_Rotation:Rotator=@n9.Rotation) -> (Output_Get:Rotator)
   n10 = BreakTransform(InTransform=GetSocketTransform(self=@n4, InSocketName=FACIAL_L_Jawline, TransformSpace=RTS_ParentBoneSpace))
   > Set FACIAL_L_Jawline_Translation(FACIAL_L_Jawline_Translation:Vector=@n10) -> (Output_Get:Vector)
   > Set FACIAL_L_Jawline_Rotation(FACIAL_L_Jawline_Rotation:Rotator=@n10.Rotation) -> (Output_Get:Rotator)
   n11 = BreakTransform(InTransform=GetSocketTransform(self=@n4, InSocketName=FACIAL_R_Jawline, TransformSpace=RTS_ParentBoneSpace))
   > Set FACIAL_R_Jawline_Translation(FACIAL_R_Jawline_Translation:Vector=@n11) -> (Output_Get:Vector)
   > Set FACIAL_R_Jawline_Rotation(FACIAL_R_Jawline_Rotation:Rotator=@n11.Rotation) -> (Output_Get:Rotator)
   > n12 = Gravity Override State Function DEPRECATED(
    CharacterPartAnimInstance=Self-Reference()
    GameTimeInSeconds=GetGameTimeInSeconds()
    GravityOverrideParams=Get Gravity Override Params_0() -> (Gravity Override Params_0:Gravity Override Params Struct Structure)
   )
   > Set Override World Gravity(Override World Gravity:Vector=@n12) -> (Output_Get:Vector)
   n13 = OR Boolean!(A=Get bIsCrouching(), B=Get bIsCrouchSprinting(), C:bool=Get bIsJumping(), D:bool=Get bIsFalling()) -> (ReturnValue)
   n14 = Dot Product!(
    A:Vector=GetActorForwardVector!(self:Actor Object Reference=TryGetPawnOwner()) -> (ReturnValue:Vector)
    B:Vector=Normalize!(A:Vector=Get PawnVelocity(), Tolerance:float=0.0001) -> (ReturnValue:Vector) {is=title K2Node_CallFunction}
   ) -> (ReturnValue:double)
   > Set Discman_on_Moving(Discman_on_Moving:double=SelectFloat(A=MapRangeClamped(Value=@n14, InRangeA=0, InRangeB=1), B=0, bPickA=NOT Boolean(A=@n13))) -> (Output_Get:double)
   > Set JumpandFall(JumpandFall:bool=OR Boolean!(A=Get bIsJumping(), B=Get bIsFalling()) -> (ReturnValue)) -> (Output_Get:bool)
   n15 = Event OnCharacterCustomizationLoaded()
   n16 = GetSkinnedAsset!(self:Skinned Mesh Component Object Reference=Get Array Item with EFort Custom Part Type(Array=@n15.CustomMeshParts, PartType=Face)) -> (ReturnValue:Skinned Asset Object Reference)
   > Set Helmet_on(Helmet_on:bool="Equal Exactly (String)"(A=GetDisplayName(Object=@n16), B=Get isHelmet() -> (isHelmet:String))) -> (Output_Get:bool)
   > Set BodyMesh(BodyMesh:Skeletal Mesh Component Budgeted Object Reference=Get Array Item with EFort Custom Part Type(Array=@n15.CustomMeshParts, PartType=Head)) -> (Output_Get:Skeletal Mesh Component Budgeted Object Reference)
  ]}
]
variables [
 BodyMesh {type=SkeletalMeshComponentBudgeted}
 Discman_on_Moving {type=number}
 FACIAL_C_Jaw_Rotation {type=Rotator, default=<rot 0 0 20>}
 FACIAL_C_Jaw_Tranform {type=Vector}
 FACIAL_C_Jawline_Rotation {type=Rotator, default=<rot 0 0 20>}
 FACIAL_C_Jawline_Translation {type=Vector}
 FACIAL_C_UnderChin_Rotation {type=Rotator, default=<rot 0 0 20>}
 FACIAL_C_UnderChin_Translation {type=Vector}
 FACIAL_L_Jawline_Rotation {type=Rotator, default=<rot 0 0 20>}
 FACIAL_L_Jawline_Translation {type=Vector}
 FACIAL_L_UnderChin_Rotation {type=Rotator, default=<rot 0 0 20>}
 FACIAL_L_UnderChin_Translation {type=Vector}
 FACIAL_R_Jawline_Rotation {type=Rotator, default=<rot 0 0 20>}
 FACIAL_R_Jawline_Translation {type=Vector}
 FACIAL_R_UnderChin_Rotation {type=Rotator, default=<rot 0 0 20>}
 FACIAL_R_UnderChin_Translation {type=Vector}
 Game Time in Seconds_0 {type=number}
 Gravity Override Params_0 {type=GravityOverrideParamsStruct}
 Helmet_on {type=boolean}
 isFaceBones {type=boolean}
 isHelmet {type=string, default=F_MED_ShiverFlame_Ace_FaceAcc}
 JumpandFall {type=boolean}
 Override World Gravity {type=Vector}
]
