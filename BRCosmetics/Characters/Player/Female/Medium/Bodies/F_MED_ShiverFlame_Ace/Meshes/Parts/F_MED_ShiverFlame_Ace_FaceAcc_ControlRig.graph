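# Control Rig for the face accessory part of the female ShiverFlame Ace character, driving the ponytail bone chain.
# Translations of the ponytail master bone and three cst_ponytail bones form a four-point spline.
# The five dyn_ponytail bones are fitted to that spline with stretched alignment along X, propagating to children.
# The first cst bone follows the head while the other two follow spine_05, so the ponytail bends between head and upper back.
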ControlRigBlueprint /BRCosmetics/Characters/Player/Female/Medium/Bodies/F_MED_ShiverFlame_Ace/Meshes/Parts/F_MED_ShiverFlame_Ace_FaceAcc_ControlRig
graphs [
 RigVMFunctionLibrary {links [], nodes []}
 RigVMModel {nodes [
   RigUnit_BeginExecution = &RigUnit_BeginExecution()
   RigUnit_GetTransform = &RigUnit_GetTransform(Item=(Type=Bone,Name="cst_ponytail_03"))
   SplineThrough4Points = RigVMFunctionReferenceNode(
    Execute=@RigUnit_BeginExecution.ExecutePin
    Point 3=@RigUnit_GetTransform.Transform.Translation
    Point 0=&RigUnit_GetTransform(Item=(Type=Bone,Name="dyn_ponytail_MSTR")).Transform.Translation
    Point 1=&RigUnit_GetTransform(Item=(Type=Bone,Name="cst_ponytail_01")).Transform.Translation
    Point 2=&RigUnit_GetTransform(Item=(Type=Bone,Name="cst_ponytail_02")).Transform.Translation
   )
   DISPATCH_RigVMDispatch_ArrayMake = RigVMDispatchNode(Values=((Type=Bone,Name="dyn_ponytail_01"),((),Type=Bone,Name="dyn_ponytail_02"),((),Type=Bone,Name="dyn_ponytail_03"),((),Type=Bone,Name="dyn_ponytail_04"),((),Type=Bone,Name="dyn_ponytail_05"))) {template="DISPATCH_RigVMDispatch_ArrayMake(in Values,out Array)"}
   /Script/ControlRigSpline.RigUnit_FitChainToSplineCurveItemArray(
    Alignment=Stretched
    DebugSettings=(bEnabled=False,Scale=1,CurveColor=(R=1,G=1,B=0,A=1),SegmentsColor=(R=1,G=0,B=0,A=1),WorldOffset=(Rotation=(X=0,Y=0,Z=0,W=1),Translation=(X=0,Y=0,Z=0),Scale3D=(X=1,Y=1,Z=1)))
    Maximum=1
    Minimum=0
    PoleVectorPosition=(X=0,Y=0,Z=0)
    PrimaryAxis=(X=1,Y=0,Z=0)
    RotationEaseType=Linear
    Rotations=()
    SamplingPrecision=12
    SecondaryAxis=(X=0,Y=0,Z=0)
    Weight=1
    bPropagateToChildren=True
    Items=@DISPATCH_RigVMDispatch_ArrayMake.Array
    ExecutePin=@SplineThrough4Points.Execute
    Spline=@SplineThrough4Points.Spline
   ) {is=struct RigVMUnitNode}
  ]}
]
hierarchy [
 root {loc=<0 0 0>, rot=<0 0 0 1>}
 pelvis {parent=root, loc=<0 -0.5301 98.0571>, rot=<0.0127 0.707 -0.0127 -0.707>}
 spine_01 {parent=pelvis, loc=<6.1208 -0.9328 0>, rot=<0 0 0.034 -0.9994>}
 spine_02 {parent=spine_01, loc=<6.5996 -0.2095 0>, rot=<0 0 0.0238 -0.9997>}
 spine_03 {parent=spine_02, loc=<6.4914 -0.1237 0>, rot=<0 0 0.0149 -0.9999>}
 spine_04 {parent=spine_03, loc=<6.9598 0.0794 0>, rot=<0 0 -0.0132 -0.9999>}
 spine_05 {parent=spine_04, loc=<6.9987 0 0>, rot=<0 0 -0.0015 -1>}
 clavicle_l {parent=spine_05, loc=<5.4831 -0.1053 -3.2946>, rot=<0.0465 -0.7289 -0.009 -0.6829>}
 neck_01 {parent=spine_05, loc=<11.3138 0.1529 0>, rot=<0 0 -0.2003 -0.9797>}
 neck_02 {parent=neck_01, loc=<4.5627 0 0>, rot=<0 0 0.0504 -0.9987>}
 head {parent=neck_02, loc=<5.3914 0 0>, rot=<0 0 0.115 -0.9934>}
 dyn_ponytail_MSTR {parent=head, loc=<11.571 -5.8356 -0.0612>, rot=<0.4938 -0.5061 0.5061 -0.4938>}
 dyn_ponytail_01 {parent=dyn_ponytail_MSTR, loc=<0 0 0>, rot=<0 0 0 1>}
 dyn_ponytail_02 {parent=dyn_ponytail_01, loc=<5.5 0 0>, rot=<0 -0.5036 0 -0.8639>}
 dyn_ponytail_03 {parent=dyn_ponytail_02, loc=<5.5 0 0>, rot=<0 -0.3641 0 -0.9314>}
 dyn_ponytail_04 {parent=dyn_ponytail_03, loc=<10 0 0>, rot=<0 0.0937 0 -0.9956>}
 dyn_ponytail_05 {parent=dyn_ponytail_04, loc=<10 0 0>, rot=<0 0.3021 0 -0.9533>}
 cst_ponytail_01 {parent=head, loc=<10.7359 -11.2718 -0.0612>, rot=<0.0029 -0.7071 -0.0029 -0.7071>}
 clavicle_r {parent=spine_05, loc=<5.4831 -0.1053 3.2946>, rot=<0.7289 0.0465 -0.6829 0.009>}
 cst_ponytail_02 {parent=spine_05, loc=<17.7889 -10.5708 -0.0612>, rot=<0.0284 -0.7065 -0.0284 -0.7065>}
 cst_ponytail_03 {parent=spine_05, loc=<7.7688 -10.5174 -0.0612>, rot=<0.0284 -0.7065 -0.0284 -0.7065>}
]
