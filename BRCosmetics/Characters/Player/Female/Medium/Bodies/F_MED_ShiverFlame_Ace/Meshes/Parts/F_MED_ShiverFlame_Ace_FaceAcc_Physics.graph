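# Physics asset for the face accessory part of the female medium ShiverFlame Ace character, built around a dynamic ponytail.
# Capsules cover a five-bone ponytail chain under a master bone, plus the head, neck, upper spine, clavicles and upper arms, likely as hair colliders.
# Ponytail joints have tight twist limits and swing limits that widen toward the tip.
# The master-to-head and body joints leave swing and twist free.
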PhysicsAsset /BRCosmetics/Characters/Player/Female/Medium/Bodies/F_MED_ShiverFlame_Ace/Meshes/Parts/F_MED_ShiverFlame_Ace_FaceAcc_Physics
bodies {
 clavicle_l [{shapeType=Capsule, length=7, radius=6, shapeName=clavicle_l_capsule, center=<5.4236 0 -1.5>, rotation=<rot 80 0 0>}]
 clavicle_r [{shapeType=Capsule, length=7, radius=6.5, shapeName=clavicle_r_capsule, center=<-5.7 0 1.5>, rotation=<rot 80 0 0>}]
 dyn_ponytail_01 [{shapeType=Capsule, length=4, radius=3, shapeName=dyn_ponytail_01_capsule, center=<4.0944 0 0>, rotation=<rot -90 180 180>}]
 dyn_ponytail_02 [{shapeType=Capsule, length=3, radius=4, shapeName=dyn_ponytail_02_capsule, center=<5.219 0 0>, rotation=<rot -10.8717 95.4847 -90.0009>}]
 dyn_ponytail_03 [{shapeType=Capsule, length=5, radius=4.5, shapeName=dyn_ponytail_03_capsule, center=<7 0 0>, rotation=<rot -90 0 0>}]
 dyn_ponytail_04 [{shapeType=Capsule, length=6, radius=3.5, shapeName=dyn_ponytail_04_capsule, center=<7 0 0>, rotation=<rot -70.0002 0 0>}]
 dyn_ponytail_05 [{shapeType=Capsule, length=1.5, radius=3, shapeName=dyn_ponytail_05_capsule, center=<5.7098 0 0>, rotation=<rot -90 0 0>}]
 dyn_ponytail_MSTR [{shapeType=Capsule, length=0.505, radius=0.505, shapeName=dyn_ponytail_MSTR_capsule, rotation=<rot 0 -90 0>}]
 head [{
   shapeType=Capsule
   length=2.3606
   radius=10.6331
   shapeName=head_capsule
   center=<4.6821 2.4795 -0.0024>
   rotation=<rot -0.1348 84.9553 -90.2334>
  }]
 neck_01 [{shapeType=Capsule, length=34, radius=5.5, shapeName=neck_01_capsule, center=<3.0492 0.5452 0>}]
 neck_02 [{shapeType=Capsule, length=3, radius=5.005, shapeName=neck_02_capsule, center=<1.3204 0 0>, rotation=<rot 0 -90.0002 90>}]
 spine_04 [{shapeType=Capsule, length=6, radius=9, shapeName=spine_04_capsule, center=<-4.5276 3.1958 0>}]
 spine_05 [{shapeType=Capsule, length=8, radius=10, shapeName=spine_05_capsule, center=<0.2108 0 0>}]
 upperarm_l [{shapeType=Capsule, length=24, radius=6, shapeName=upperarm_l_capsule, center=<12.3515 0.2207 0.8734>, rotation=<rot 90 -166.7373 -166.7382>}]
 upperarm_r [{
   shapeType=Capsule
   length=24
   radius=6
   shapeName=upperarm_r_capsule
   center=<-12.4425 -0.3305 -0.9226>
   rotation=<rot 90 -179.9801 -179.9804>
  }]
}
constraints [
 {bone1Name=dyn_ponytail_01, bone2Name=dyn_ponytail_MSTR, swing1LimitDegrees=10, swing2LimitDegrees=15, twistLimitDegrees=2.5}
 {bone1Name=dyn_ponytail_02, bone2Name=dyn_ponytail_01, swing1LimitDegrees=10, swing2LimitDegrees=20, twistLimitDegrees=2.5}
 {bone1Name=dyn_ponytail_03, bone2Name=dyn_ponytail_02, swing1LimitDegrees=15, swing2LimitDegrees=25, twistLimitDegrees=2.5}
 {bone1Name=dyn_ponytail_04, bone2Name=dyn_ponytail_03, swing1LimitDegrees=20, swing2LimitDegrees=35, twistLimitDegrees=2.5}
 {bone1Name=dyn_ponytail_05, bone2Name=dyn_ponytail_04, swing1LimitDegrees=25, swing2LimitDegrees=45, twistLimitDegrees=2.5}
 {
  bone1Name=dyn_ponytail_MSTR
  bone2Name=head
  swing1LimitDegrees=45
  swing1Motion=Free
  swing2LimitDegrees=45
  swing2Motion=Free
  twistLimitDegrees=45
  twistMotion=Free
 }
 {
  bone1Name=neck_02
  bone2Name=neck_01
  swing1LimitDegrees=45
  swing1Motion=Free
  swing2LimitDegrees=45
  swing2Motion=Free
  twistLimitDegrees=45
  twistMotion=Free
 }
 {
  bone1Name=head
  bone2Name=neck_02
  swing1LimitDegrees=45
  swing1Motion=Free
  swing2LimitDegrees=45
  swing2Motion=Free
  twistLimitDegrees=45
  twistMotion=Free
 }
 {
  bone1Name=clavicle_l
  bone2Name=spine_05
  swing1LimitDegrees=45
  swing1Motion=Free
  swing2LimitDegrees=45
  swing2Motion=Free
  twistLimitDegrees=45
  twistMotion=Free
 }
 {
  bone1Name=clavicle_r
  bone2Name=spine_05
  swing1LimitDegrees=45
  swing1Motion=Free
  swing2LimitDegrees=45
  swing2Motion=Free
  twistLimitDegrees=45
  twistMotion=Free
 }
 {
  bone1Name=spine_04
  bone2Name=spine_05
  swing1LimitDegrees=45
  swing1Motion=Free
  swing2LimitDegrees=45
  swing2Motion=Free
  twistLimitDegrees=45
  twistMotion=Free
 }
 {
  bone1Name=upperarm_l
  bone2Name=clavicle_l
  swing1LimitDegrees=45
  swing1Motion=Free
  swing2LimitDegrees=45
  swing2Motion=Free
  twistLimitDegrees=45
  twistMotion=Free
 }
 {
  bone1Name=upperarm_r
  bone2Name=clavicle_r
  swing1LimitDegrees=45
  swing1Motion=Free
  swing2LimitDegrees=45
  swing2Motion=Free
  twistLimitDegrees=45
  twistMotion=Free
 }
]
