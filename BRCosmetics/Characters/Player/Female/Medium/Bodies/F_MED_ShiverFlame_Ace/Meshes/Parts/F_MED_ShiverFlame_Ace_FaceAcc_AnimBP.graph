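# Post-process style AnimBP for the ShiverFlame Ace face accessory, animating the ponytail on top of the pose copied from the attached parent.
# Upper arms are overwritten in world space from body mesh sockets, then cached as the main pose.
# Without back bling, RigidBody physics simulates from neck_01 with skydive-aware gravity override.
# With back bling or cape, stacked ponytail bone offsets and scales feed a Control Rig; head rotation factors are also computed each update.
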AnimBlueprint /BRCosmetics/Characters/Player/Female/Medium/Bodies/F_MED_ShiverFlame_Ace/Meshes/Parts/F_MED_ShiverFlame_Ace_FaceAcc_AnimBP
target_skeleton=&/F_MED_ShiverFlame_Ace_FaceAcc_Skeleton
graphs [
 AnimGraph {nodes [
   n1 = RigidBody!(ComponentPose=Local To Component(LocalPose=Use cached pose 'main pose'()), Alpha, OverrideWorldGravity:Vector=Get Gravity Override()) -> (Pose) {
    bClampLinearTranslationLimitToRefPose=true
    bOverrideWorldGravity=true
    overridePhysicsAsset=&/F_MED_ShiverFlame_Ace_FaceAcc_Physics
    simulationSpace=BaseBoneSpace
    alphaBoolBlend {blendOption=Linear}
    baseBoneRef {boneName=neck_01}
    componentAppliedLinearAccClamp=<800 800 800>
    componentLinearAccScale=<0.5 0.5 0.5>
    componentLinearVelScale=<0.5 0.5 0.5>
   }
   Output Pose(Result=Blend Poses by bool(
     BlendPose_0=Use cached pose 'backpack pose'()
     BlendPose_1=Component To Local(ComponentPose=@n1)
     bActiveValue=Get isWearingBackpack() -> (isWearingBackpack:bool)
    ) {blendType=Linear}) {layerGroup=None}
   n2 = "Transform (Modify) Bone - Bone: upperarm_l"(
    ComponentPose=Local To Component(LocalPose=Copy Pose From Mesh() {bUseAttachedParent=true})
    Alpha=1
    Translation=Get upperarm l translation() -> (upperarm l translation:Vector)
    Rotation=Get upperarm l rotation() -> (upperarm l rotation:Rotator)
    Scale="1,1,1"
   ) {
    rotationMode=BMM_Replace
    rotationSpace=BCS_WorldSpace
    translationMode=BMM_Replace
    translationSpace=BCS_WorldSpace
    alphaBoolBlend {blendOption=Linear}
    boneToModify {boneName=upperarm_l}
   }
   n3 = "Transform (Modify) Bone - Bone: upperarm_r"(
    ComponentPose=@n2
    Alpha=1
    Translation=Get upperarm r translation() -> (upperarm r translation:Vector)
    Rotation=Get upperarm r rotation() -> (upperarm r rotation:Rotator)
    Scale="1,1,1"
   ) {
    rotationMode=BMM_Replace
    rotationSpace=BCS_WorldSpace
    translationMode=BMM_Replace
    translationSpace=BCS_WorldSpace
    alphaBoolBlend {blendOption=Linear}
    boneToModify {boneName=upperarm_r}
   }
   Save cached pose 'main pose'(Pose=Component To Local(ComponentPose=@n3)) {cacheName=main pose}
   n4 = "Transform (Modify) Bone - Bone: cst_ponytail_01"!(ComponentPose=Local To Component(LocalPose=Use cached pose 'main pose'()), Alpha=1, Translation="7,0,0") -> (Pose) {
    rotationSpace=BCS_BoneSpace
    translationMode=BMM_Additive
    translationSpace=BCS_BoneSpace
    alphaBoolBlend {blendOption=Linear}
    boneToModify {boneName=cst_ponytail_01}
    translation {x=7}
   }
   n5 = "Transform (Modify) Bone - Bone: cst_ponytail_02"!(ComponentPose=@n4, Alpha=1, Translation="0,0,0") -> (Pose) {
    rotationSpace=BCS_BoneSpace
    translationMode=BMM_Additive
    translationSpace=BCS_BoneSpace
    alphaBoolBlend {blendOption=Linear}
    boneToModify {boneName=cst_ponytail_02}
   }
   n6 = "Transform (Modify) Bone - Bone: cst_ponytail_03"!(ComponentPose=@n5, Alpha=1, Translation="0,0,0") -> (Pose) {
    rotationSpace=BCS_BoneSpace
    translationMode=BMM_Additive
    translationSpace=BCS_BoneSpace
    alphaBoolBlend {blendOption=Linear}
    boneToModify {boneName=cst_ponytail_03}
   }
   n7 = "Transform (Modify) Bone - Bone: cst_ponytail_01"(ComponentPose=@n6, Alpha=1, Translation="-7,0,0") {
    rotationSpace=BCS_BoneSpace
    translationMode=BMM_Additive
    translationSpace=BCS_BoneSpace
    alphaBoolBlend {blendOption=Linear}
    boneToModify {boneName=cst_ponytail_01}
    translation {x=-7}
   }
   n8 = "Transform (Modify) Bone - Bone: cst_ponytail_02"(ComponentPose=@n7, Alpha=1, Translation="0,0,0") {
    rotationSpace=BCS_BoneSpace
    translationMode=BMM_Additive
    translationSpace=BCS_BoneSpace
    alphaBoolBlend {blendOption=Linear}
    boneToModify {boneName=cst_ponytail_02}
   }
   n9 = "Transform (Modify) Bone - Bone: cst_ponytail_03"(ComponentPose=@n8, Alpha=1, Translation="0,0,0") {
    rotationSpace=BCS_BoneSpace
    translationMode=BMM_Additive
    translationSpace=BCS_BoneSpace
    alphaBoolBlend {blendOption=Linear}
    boneToModify {boneName=cst_ponytail_03}
   }
   n10 = "Transform (Modify) Bone - Bone: cst_ponytail_01"(ComponentPose=@n9, Alpha=1, Translation="0,0,0") {
    rotationSpace=BCS_BoneSpace
    translationMode=BMM_Additive
    translationSpace=BCS_BoneSpace
    alphaBoolBlend {blendOption=Linear}
    boneToModify {boneName=cst_ponytail_01}
   }
   n11 = "Transform (Modify) Bone - Bone: cst_ponytail_02"(ComponentPose=@n10, Alpha=1, Translation="-8.735,0,0") {
    rotationSpace=BCS_BoneSpace
    translationMode=BMM_Additive
    translationSpace=BCS_BoneSpace
    alphaBoolBlend {blendOption=Linear}
    boneToModify {boneName=cst_ponytail_02}
    translation {x=-8.735}
   }
   n12 = "Transform (Modify) Bone - Bone: cst_ponytail_03"(ComponentPose=@n11, Alpha=1, Translation="-11.193,0,0") {
    rotationSpace=BCS_BoneSpace
    translationMode=BMM_Additive
    translationSpace=BCS_BoneSpace
    alphaBoolBlend {blendOption=Linear}
    boneToModify {boneName=cst_ponytail_03}
    translation {x=-11.193}
   }
   n13 = "Transform (Modify) Bone - Bone: cst_ponytail_01"(ComponentPose=@n12, Alpha=1, Translation="0,0,1.772") {
    rotationSpace=BCS_BoneSpace
    translationMode=BMM_Additive
    translationSpace=BCS_BoneSpace
    alphaBoolBlend {blendOption=Linear}
    boneToModify {boneName=cst_ponytail_01}
    translation {z=1.772}
   }
   n14 = "Transform (Modify) Bone - Bone: cst_ponytail_02"(ComponentPose=@n13, Alpha=1, Translation="-3.058,1.593,-3.526") {
    rotationSpace=BCS_BoneSpace
    translationMode=BMM_Additive
    translationSpace=BCS_BoneSpace
    alphaBoolBlend {blendOption=Linear}
    boneToModify {boneName=cst_ponytail_02}
    translation=<-3.058 1.593 -3.526>
   }
   n15 = "Transform (Modify) Bone - Bone: cst_ponytail_03"(ComponentPose=@n14, Alpha=1, Translation="-2.296,2.348,0.000") {
    rotationSpace=BCS_BoneSpace
    translationMode=BMM_Additive
    translationSpace=BCS_BoneSpace
    alphaBoolBlend {blendOption=Linear}
    boneToModify {boneName=cst_ponytail_03}
    translation=<-2.296 2.348>
   }
   n16 = "Transform (Modify) Bone - Bone: cst_ponytail_01"(ComponentPose=@n15, Alpha=1, Translation="0.000,4.756,4.077") {
    rotationSpace=BCS_BoneSpace
    translationMode=BMM_Additive
    translationSpace=BCS_BoneSpace
    alphaBoolBlend {blendOption=Linear}
    boneToModify {boneName=cst_ponytail_01}
    translation {y=4.756, z=4.077}
   }
   n17 = "Transform (Modify) Bone - Bone: cst_ponytail_02"(ComponentPose=@n16, Alpha=1, Translation="4.257,-3.724,-15.167") {
    rotationSpace=BCS_BoneSpace
    translationMode=BMM_Additive
    translationSpace=BCS_BoneSpace
    alphaBoolBlend {blendOption=Linear}
    boneToModify {boneName=cst_ponytail_02}
    translation=<4.257 -3.724 -15.167>
   }
   n18 = "Transform (Modify) Bone - Bone: cst_ponytail_03"(ComponentPose=@n17, Alpha=1, Translation="4.124,-1.701,-14.826") {
    rotationSpace=BCS_BoneSpace
    translationMode=BMM_Additive
    translationSpace=BCS_BoneSpace
    alphaBoolBlend {blendOption=Linear}
    boneToModify {boneName=cst_ponytail_03}
    translation=<4.124 -1.701 -14.826>
   }
   n19 = "Transform (Modify) Bone - Bone: dyn_ponytail_01"!(ComponentPose=@n18, Alpha=1, Scale="1,1,0.8") -> (Pose) {
    rotationSpace=BCS_BoneSpace
    scaleMode=BMM_Replace
    scaleSpace=BCS_BoneSpace
    translationSpace=BCS_BoneSpace
    alphaBoolBlend {blendOption=Linear}
    boneToModify {boneName=dyn_ponytail_01}
    rotation {pitch=10}
    scale {z=0.8}
   }
   n20 = "Transform (Modify) Bone - Bone: dyn_ponytail_02"!(ComponentPose=@n19, Alpha=1, Scale="1,1,0.7") -> (Pose) {
    rotationSpace=BCS_BoneSpace
    scaleMode=BMM_Replace
    scaleSpace=BCS_BoneSpace
    translationSpace=BCS_BoneSpace
    alphaBoolBlend {blendOption=Linear}
    boneToModify {boneName=dyn_ponytail_02}
    scale {z=0.7}
   }
   n21 = "Transform (Modify) Bone - Bone: dyn_ponytail_03"!(ComponentPose=@n20, Alpha=1, Scale="1,1,1.2") -> (Pose) {
    rotationSpace=BCS_BoneSpace
    scaleMode=BMM_Replace
    scaleSpace=BCS_BoneSpace
    translationSpace=BCS_BoneSpace
    alphaBoolBlend {blendOption=Linear}
    boneToModify {boneName=dyn_ponytail_03}
    scale {z=1.2}
   }
   n22 = "Transform (Modify) Bone - Bone: dyn_ponytail_04"!(ComponentPose=@n21, Alpha=1, Scale="1,1,1.3") -> (Pose) {
    rotationSpace=BCS_BoneSpace
    scaleMode=BMM_Replace
    scaleSpace=BCS_BoneSpace
    translationSpace=BCS_BoneSpace
    alphaBoolBlend {blendOption=Linear}
    boneToModify {boneName=dyn_ponytail_04}
    scale {z=1.3}
   }
   n23 = "Transform (Modify) Bone - Bone: dyn_ponytail_05"!(ComponentPose=@n22, Alpha=1, Scale="1,1,1") -> (Pose) {
    rotationSpace=BCS_BoneSpace
    scaleMode=BMM_Replace
    scaleSpace=BCS_BoneSpace
    translationSpace=BCS_BoneSpace
    alphaBoolBlend {blendOption=Linear}
    boneToModify {boneName=dyn_ponytail_05}
   }
   n24 = Control Rig(Source=Component To Local(ComponentPose=@n23)) {
    alphaBoolBlend {blendOption=Linear}
    controlRigAssetReference {blueprintRigClass=&/F_MED_ShiverFlame_Ace_FaceAcc_ControlRig.F_MED_ShiverFlame_Ace_FaceAcc_ControlRig_C}
   }
   Save cached pose 'backpack pose'(Pose=@n24) {cacheName=backpack pose}
  ]}
 EventGraph {nodes [
   Event BlueprintUpdateAnimation()
   > n1 = Sequence!(execute) -> (then_0, then_1, then_2:Exec)
   n2 = Get BodyMesh() -> (BodyMesh:Skeletal Mesh Component Object Reference)
   Is Valid(exec=@n1.then_2, InputObject=@n2)
   n3 = BreakTransform(InTransform=GetSocketTransform(self=@n2, InSocketName=upperarm_l, TransformSpace=RTS_World))
   > Set upperarm l translation(upperarm l translation:Vector=@n3) -> (Output_Get:Vector)
   > Set upperarm l rotation(upperarm l rotation:Rotator=@n3.Rotation) -> (Output_Get:Rotator)
   n4 = BreakTransform(InTransform=GetSocketTransform(self=@n2, InSocketName=upperarm_r, TransformSpace=RTS_World))
   > Set upperarm r translation(upperarm r translation:Vector=@n4) -> (Output_Get:Vector)
   > Set upperarm r rotation(upperarm r rotation:Rotator=@n4.Rotation) -> (Output_Get:Rotator)
   n5 = CheckForBackbling()
   Set isWearingBackpack(execute=@n1.then_0, isWearingBackpack:bool=OR Boolean(A=@n5, B=@n5.bIsBackblingACape)) -> (Output_Get:bool)
   > n6 = Gravity Override State Function DEPRECATED(CharacterPartAnimInstance=Self-Reference(), GameTimeInSeconds=GetGameTimeInSeconds(), GravityOverrideParams=Get Gravity Override Params())
   > Set Gravity Override(Gravity Override=@n6)
   n7 = GetOwnerSkeletalMeshComponent()
   n8 = BreakRotator(InRot=BreakTransform(InTransform=GetDeltaTransformFromRefPose(self=@n7, BoneName=head, BaseName=spine_05)).Rotation)
   Set head left roll(execute=@n1.then_1, head left roll:double=MapRangeClamped(Value=@n8, InRangeA=0, InRangeB=70)) -> (Output_Get:double)
   > Set head right roll(head right roll:double=MapRangeClamped(Value=@n8, InRangeA=0, InRangeB=-70)) -> (Output_Get:double)
   > Set head right pitch(head right pitch:float=MapRangeClamped(Value=@n8.Pitch, InRangeA=0, InRangeB=30)) -> (Output_Get:float)
   n9 = MapRangeClamped(Value=@n8.Yaw, InRangeA=0, InRangeB=-60)
   > Set head yaw(head yaw:float=@n9) -> (Output_Get:float)
   n10 = BreakRotator(InRot=BreakTransform(InTransform=GetDeltaTransformFromRefPose(self=@n7, BoneName=spine_05, BaseName=pelvis)).Rotation)
   > Set head yaw correction(head yaw correction:double=float * float(A=@n9, B=MapRangeClamped(Value=@n10.Yaw, InRangeA=10, InRangeB=30)) {is=title K2Node_PromotableOperator}) -> (Output_Get:double)
   n11 = Event OnCharacterCustomizationLoaded()
   > Set BodyMesh(BodyMesh=Get Array Item with EFort Custom Part Type(Array=@n11.CustomMeshParts))
   PrintString!(execute, InString=Hello, bPrintToScreen, bPrintToLog, TextColor, Duration=1000, Key) -> (then)
  ]}
]
variables [
 BodyMesh {type=SkeletalMeshComponent}
 Gravity Override {type=Vector}
 Gravity Override Params {
  type=GravityOverrideParamsStruct
  default {
   skydive_FloatingDown {gravityOverride=<0 -700 -600>, windFrequency=<1.5 4 5>}
   skydive_GoingUp {gravityOverride=<0 -300 100>, windFrequency=<1.5 5 5>}
   skydive_Parachuting {gravityOverride=<0 -800 -2000>, windFrequency=<1.5 4 5>}
  }
 }
 head left roll {type=number}
 head right pitch {type=number}
 head right roll {type=number}
 head yaw {type=number}
 head yaw correction {type=number}
 isWearingBackpack {type=boolean}
 upperarm l rotation {type=Rotator}
 upperarm l translation {type=Vector}
 upperarm r rotation {type=Rotator}
 upperarm r translation {type=Vector}
]
